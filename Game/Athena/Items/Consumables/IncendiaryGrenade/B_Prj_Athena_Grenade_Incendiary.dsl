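; Incendiary grenade projectile for Athena that explodes on stopping or entering water, spawning a building gameplay actor (likely fire) unless in water.
; Half the fuse time in, a timeline ramps the fuse audio's PreExplo parameter.
; Fuse end activates a radius-sized particle effect, tinted red for non-teammates; bounces set the impact sound's surface.
; Explosion plays sound and a feedback cue, spawning at the capsule transform; the line-trace placement function is unused here.
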(event Custom|Pre%20Explo%20Audio%20Tell
  ;[1 1 1] Ramp up the audio right before explo
  (Components|Timeline|SetPlayRate (Variables|B_Prj_Athena_Grenade_Incendiary|GetPreExplosionWarning) (/ 1.0 (/ (Variables|Default|GetFuseTime) 2.0)))
  (bind (_direction _preexplo) (Timeline|PreExplosionWarning PlayFromStart)
    (:Update
      (Audio|Parameter|SetFloatParameter (Variables|B_Prj_Grenade|GetGrenadeFuse_AudioComponent) "PreExplo" _preexplo :!class /Script/AudioExtensions.AudioParameterControllerInterface))))

(event EventBeginPlay
  ;[1 1 1] Set up Fuse Timer and Bind Event on Water Enter
  (|CallParentFunction)
  (Utilities|Time|SetTimerbyEvent (EventDispatchers|CreateEvent :!event Custom|Pre%20Explo%20Audio%20Tell) (/ (Variables|Default|GetFuseTime) 2.0))
  (Water|BindEventtoOnEnterWaterEvent (Variables|Water|GetWaterInteractionComponent) (delegate Custom|On%20Enter%20Water,%20Explode)))

(event Custom|On%20Enter%20Water,%20Explode
  (params (WaterBody :type Fort%20Water%20Body%20Actor%20Object%20Reference) (WaterInteractionComponent :type Fort%20Water%20Interaction%20Component%20Object%20Reference) (bIsFirstBody :type Boolean))
  ;[1 1 1] Set up Fuse Timer and Bind Event on Water Enter
  (Delivery|DoExplosion))

(event EventFuseEnded
  (bind _effect_distance (Variables|B_Prj_Grenade|GetEffect_Distance))
  ;[1 1 1] When the fuse expires, set rotation, activate and determine colour of particle FX for team colour
  (Transformation|SetWorldRotation _effect_distance "90.000000,0.000000,0.000000" false false)
  (Components|Activation|Activate _effect_distance false)
  (Effects|Components|ParticleSystem|SetVectorParameter _effect_distance "Size" (Math|Conversions|ToVector(Float) (Variables|Default|GetExplosionRadius)))
  (if (not (Team|OnSameTeam (Game|GetInstigatorController) (Game|GetPlayerController 0)))
    (Effects|Components|ParticleSystem|SetVectorParameter _effect_distance "Color" "1.500000,0.110000,0.190000")))

(event EventOnStop
  (params (Hit :type Hit%20Result%20Structure%20%28by%20ref%29))
  ;[1 1 1] On Stop, Explode (This is to prevent the grenade exploding mid-air and creating a burning floating patch)
  (Delivery|DoExplosion)
  (Audio|Components|Audio|FadeOut (Variables|B_Prj_Grenade|GetGrenadeFuse_AudioComponent) 0.3 0.0 "Linear" :!class /Script/Engine.AudioComponent))

(event EventOnExploded
  (params (HitActors :type Array%20of%20Actor%20Object%20References) (HitResults :type Array%20of%20Hit%20Result%20Structures))
  ;[1 1 1] On Explode: If in Water, skip Line Trace & BGA Spawn. Otherwise do a line-trace to ensure we spawn the BGA correctly.
  (|CallParentFunction HitActors HitResults)
  (if (not (Water|IsinWater (Variables|Water|GetWaterInteractionComponent)))
    (Building|SpawnBuildingGameplayActor (Variables|Default|GetSpawn%20Class) (Math|Transform|MakeTransform (Transformation|GetWorldLocation (Variables|Collision|GetCapsuleComponent)) (Transformation|GetWorldRotation (Variables|Collision|GetCapsuleComponent))) (Game|GetInstigator)))
  (Class|BPSurfaceTypeSoundComponent|PlaySound (Variables|Default|GetBP_SurfaceTypeSoundComponent) (Transformation|GetActorLocation))
  (GameplayCue|ExecuteGameplayCue%28Burst%29Local self (Variables|Default|GetFeedbackCue) (Ability|GameplayCue|MakeGameplayCueParameters 0.0)) ; Cam shake + controller vibration
)

(event EventOnBounce
  (params (Hit :type Hit%20Result%20Structure%20%28by%20ref%29))
  (bind (_bblockinghit _binitialoverlap _time _distance _location _impactpoint _normal _impactnormal _physmat _hitactor) (Collision|BreakHitResult Hit))
  ;[1 1 1] On Bounce, Set Surface Type for Audio Response
  (|CallParentFunction Hit)
  (Utilities|IsValid _physmat
    (:Is%20Valid
      (Class|BPSurfaceTypeSoundComponent|SetSurfaceType (Variables|Default|GetBP_SurfaceTypeSoundComponent) (Variables|/Script/PhysicsCore.PhysicalMaterial|GetSurfaceType _physmat) _hitactor))))

(fn LineTraceCheckforPlaneSpawn
  (bind _capsulecomponent (Variables|Collision|GetCapsuleComponent))
  (bind _returnvalue (Transformation|GetWorldLocation _capsulecomponent))
  (bind _outhit (Collision|LineTraceByChannel _returnvalue (+ _returnvalue (Variables|Default|GetFire%20Plane%20Trace%20Offset)) "TraceTypeQuery10"))
  (return (Math|Transform|SelectTransform (Math|Transform|MakeTransform (+ (.Location (Collision|BreakHitResult _outhit)) (* (.Normal (Collision|BreakHitResult _outhit)) (Variables|Default|GetFire%20Plane%20Spawn%20Offset))) (Math|Rotator|MakeRotfromZX (.Normal (Collision|BreakHitResult _outhit)) (1.0 0.0 0.0))) (Math|Transform|MakeTransform _returnvalue) (Collision|BreakHitResult _outhit))))

(fn ConstructionScript
  (|CallParentFunction))
